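; Thrown coconut projectile for the foraged coconut consumable, which reports hits on hostile pawns for quest tracking.
; A capsule component hit passes the other actor to CheckCoconutHit, which checks for the Pawn gameplay tag and a Hostile team affiliation relative to the instigator.
; A qualifying hit sends the Athena.Quest.Foraged.Hit.Coconut gameplay event to the instigator, with a DoOnce limiting it to one event per projectile.
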(event Custom|CheckCoconutHit
  (params (HitActor :type Actor%20Object%20Reference))
  (bind _returnvalue (Game|GetInstigator))
  (bind _returnvalue_2 (GameplayTags|HasAnyMatchingGameplayTags(Message) HitActor (GameplayTags|MakeLiteralGameplayTagContainer "(GameplayTags=((TagName=\"Pawn\")))")))
  (if (and _returnvalue_2 (Utilities|Enum|Equal(Enum) (Team|GetTeamAffiliation _returnvalue HitActor) "Hostile"))
    (Utilities|FlowControl|DoOnce false
      (:Completed
        (Ability|SendGameplayEventtoActor _returnvalue "(TagName=\"Athena.Quest.Foraged.Hit.Coconut\")")))))

(event OnComponentHit(CapsuleComponent)
  (params (HitComponent :type Primitive%20Component%20Object%20Reference) (OtherActor :type Actor%20Object%20Reference) (OtherComp :type Primitive%20Component%20Object%20Reference) (NormalImpulse :type Vector :default "0, 0, 0") (Hit :type Hit%20Result%20Structure%20%28by%20ref%29))
  (CallFunction|CheckCoconutHit OtherActor))

(fn ConstructionScript
  (|CallParentFunction))
